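# Material function for the EarthLane backpack FX that computes a clamped world position offset toward a Niagara-supplied position.
# It subtracts particle position from a vector built solely from dynamic parameter X (index 3).
# The difference's R component feeds both horizontal axes, with separate XY and Z clamps.
# A powered, inverted UV V gradient times Interpolation Intensity blends in the offset, weighting it along the mesh.
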MaterialFunction /BRCosmetics/Accessories/FORT_Backpacks/Backpack_M_MED_EarthLane/FX/MF_InterpolateWPO
graph {nodes [
  n1 = DynamicParameter() {desc=Parameters set by Niagara, parameterIndex=3, defaultValue=<rgba 0 0 0 0>, paramNames [X, Y, Z, Interpolation Intensity]}
  n2 = MaterialFunctionCall(Float3=Subtract(
    A=MaterialFunctionCall(X=@n1.X, Y=@n1.X, Z=@n1.X) {materialFunction=&MakeFloat3}.Result
    B=ParticlePositionWS()
   )) {materialFunction=&BreakOutFloat3Components}
  n3 = ScalarParameter() {defaultValue=100, group=Interpolate, parameterName=Interpolate Clamp XY}
  n4 = Clamp(None=AppendVector(A=@n2.R, B=@n2.R), Min=Multiply(A=@n3) {constB=-1}, Max=@n3) {maxDefault=1000, minDefault=-1000}
  n5 = ScalarParameter() {defaultValue=20, group=Interpolate, parameterName=Interpolate Clamp Z}
  n6 = Clamp(None=@n2.B, Min=Multiply(A=@n5) {constB=-1}, Max=@n5) {minDefault=-1}
  n7 = FunctionInput(Preview=TextureCoordinate()) {bUsePreviewValueAsDefault=true, inputName=UVs, inputType=FunctionInput_Vector2}
  n8 = Power(
   Base=ComponentMask(None=OneMinus(None=@n7)) {g=true}
   Exp=ScalarParameter() {defaultValue=1.7, group=Interpolate, parameterName=Interpolate Power}
  )
  FunctionOutput(None=LinearInterpolate(B=AppendVector(A=@n4, B=@n6), Alpha=Multiply(A=@n1.Interpolation Intensity, B=@n8)))
 ]}
