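# Animation Blueprint for the M_MED_EarthLane backpack cosmetic, adding spring-driven AnimDynamics motion to its lid and zipper bones.
# Three lid variants run from the cached input pose through nested bool blends into zipper simulation, limited by a plane on the pack bone.
# That result is cached and gets a final lid simulation before output.
# Each update, the skydiving state selects linear velocity and acceleration scales, which feed one blended lid variant.
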AnimBlueprint /BRCosmetics/Accessories/FORT_Backpacks/Backpack_M_MED_EarthLane/Meshes/M_MED_EarthLane_Pack_AnimBP
target_skeleton=&/M_MED_EarthLane_Pack_Skeleton
graphs [
 AnimGraph {nodes [
   n1 = Anim Dynamics - Bone: dyn_lid_01!(
    ComponentPose=Local To Component(LocalPose=Use cached pose 'SavedPose18478'())
    bAlphaBoolEnabled:bool=True
    ComponentLinearAccScale:Vector="0.15,0.15,0.15"
    ComponentLinearVelScale:Vector="0.15,0.15,0.15"
    ExternalForce="0,0,0"
   ) -> (Pose) {
    alphaInputType=Bool
    angularDampingOverride=0.999
    angularSpringConstant=900
    bAngularSpring=true
    bOverrideAngularDamping=true
    bUseGravityOverride=true
    simulationSpace=RootRelative
    alphaBoolBlend {blendOption=Linear}
    boundBone {boneName=dyn_lid_01}
    componentAppliedLinearAccClamp=<2000 2000 2000>
    componentLinearAccScale=<1 1 1>
    componentLinearVelScale=<1 1 1>
    gravityOverride {z=-980}
    physicsBodyDefinitions [{
      boundBone {boneName=dyn_lid_01}
      boxExtents=<16 16 16>
      constraintSetup {angularLimitsMax=<0 0 80>, angularLimitsMin=<0 0 -50>, angularTarget=<0 0 0>}
      localJointOffset=<-2 -3 0>
     }]
   }
   Output Pose(Result=Component To Local(ComponentPose=@n1)) {layerGroup=None}
   Save cached pose 'SavedPose25049'(Pose=Input Pose()) {cacheName=SavedPose25049}
   n2 = Anim Dynamics - Bone: dyn_lid_01!(
    ComponentPose=Local To Component(LocalPose=Use cached pose 'SavedPose25049'())
    Alpha:float=1
    ComponentLinearAccScale=Get Component Linear Acc Scale() -> (Component Linear Acc Scale:Vector)
    ComponentLinearVelScale=Get Component Linear Vel Scale() -> (Component Linear Vel Scale:Vector)
    ExternalForce="0,0,0"
   ) -> (Pose) {
    angularDampingOverride=0.999
    angularSpringConstant=900
    bAngularSpring=true
    bOverrideAngularDamping=true
    bUseGravityOverride=true
    simulationSpace=RootRelative
    alphaBoolBlend {blendOption=Linear}
    boundBone {boneName=dyn_lid_01}
    componentAppliedLinearAccClamp=<2000 2000 2000>
    componentLinearAccScale=<1 1 1>
    componentLinearVelScale=<1 1 1>
    gravityOverride {z=-980}
    physicsBodyDefinitions [{
      boundBone {boneName=dyn_lid_01}
      boxExtents=<16 16 16>
      constraintSetup {angularLimitsMax=<0 0 80>, angularLimitsMin=<0 0 -50>, angularTarget=<0 0 0>}
      localJointOffset=<-2 -3 0>
     }]
   }
   n3 = Anim Dynamics - Bone: dyn_lid_01(
    ComponentPose=Local To Component(LocalPose=Use cached pose 'SavedPose25049'())
    Alpha=1
    ComponentLinearAccScale="0.5,0.5,0.7"
    ComponentLinearVelScale="0.5,0.5,0.7"
    ExternalForce="0,0,0"
   ) {
    angularDampingOverride=0.9999
    angularSpringConstant=400
    bAngularSpring=true
    bOverrideAngularDamping=true
    bUseGravityOverride=true
    simulationSpace=RootRelative
    alphaBoolBlend {blendOption=Linear}
    boundBone {boneName=dyn_lid_01}
    componentAppliedLinearAccClamp=<2000 2000 2000>
    componentLinearAccScale=<1 1 1>
    componentLinearVelScale=<1 1 1>
    gravityOverride {z=-980}
    physicsBodyDefinitions [{
      boundBone {boneName=dyn_lid_01}
      boxExtents=<15 15 15>
      constraintSetup {angularLimitsMax=<0 0 60>, angularLimitsMin=<0 0 -30>, angularTarget=<0 0 0>}
      localJointOffset=<-0.5 0 0>
     }]
   }
   n4 = Blend Poses by bool(BlendPose_0=Component To Local(ComponentPose=@n3), BlendPose_1=Component To Local(ComponentPose=@n2)) {blendType=Linear}
   n5 = Anim Dynamics - Bone: dyn_lid_01(
    ComponentPose=Local To Component(LocalPose=Use cached pose 'SavedPose25049'())
    Alpha=1
    ComponentLinearAccScale="0.5,0.5,0.7"
    ComponentLinearVelScale="0.5,0.5,0.7"
    ExternalForce="0,0,0"
   ) {
    angularDampingOverride=0.9999
    angularSpringConstant=400
    bAngularSpring=true
    bOverrideAngularDamping=true
    bUseGravityOverride=true
    simulationSpace=RootRelative
    alphaBoolBlend {blendOption=Linear}
    boundBone {boneName=dyn_lid_01}
    componentAppliedLinearAccClamp=<2000 2000 2000>
    componentLinearAccScale=<1 1 1>
    componentLinearVelScale=<1 1 1>
    gravityOverride {z=-980}
    physicsBodyDefinitions [{
      boundBone {boneName=dyn_lid_01}
      boxExtents=<15 15 15>
      constraintSetup {angularLimitsMax=<0 0 60>, angularLimitsMin=<0 0 -30>, angularTarget=<0 0 0>}
      localJointOffset=<1 0 0>
     }]
   }
   n6 = Component To Local(ComponentPose=@n5)
   n7 = Anim Dynamics - Bone: dyn_zipper_01(
    ComponentPose=Local To Component(LocalPose=Blend Poses by bool(BlendPose_0=@n6, BlendPose_1=@n4) {blendType=Linear})
    Alpha=1
    ExternalForce="0,0,0"
   ) {
    angularDampingOverride=0.9
    angularSpringConstant=350
    bAngularSpring=true
    bOverrideAngularDamping=true
    bOverrideLinearDamping=true
    bUseGravityOverride=true
    linearDampingOverride=0.9
    simulationSpace=RootRelative
    alphaBoolBlend {blendOption=Linear}
    boundBone {boneName=dyn_zipper_01}
    componentAppliedLinearAccClamp=<2000 2000 2000>
    componentLinearAccScale=<0.3 0.3 0.3>
    componentLinearVelScale=<0.3 0.3 0.3>
    gravityOverride {z=-980}
    physicsBodyDefinitions [{
      boundBone {boneName=dyn_zipper_01}
      boxExtents=<5 5 5>
      constraintSetup {angularLimitsMax=<0 10 45>, angularLimitsMin=<0 -90 -45>, angularTarget=<0 0 0>, linearAxesMax=<0.2 0 0>, linearAxesMin=<-0.2 0 0>}
      localJointOffset=<0.5 0 0>
     }]
    planarLimits [{drivingBone {boneName=pack}, planeTransform {location=<0 -6.5733 -0.0505>, rotation=<rot 0 0 -90.4402>}}]
   }
   Save cached pose 'SavedPose18478'(Pose=Component To Local(ComponentPose=@n7)) {cacheName=SavedPose18478}
  ]}
 EventGraph {nodes [
   Event BlueprintUpdateAnimation()
   n1 = Get bIsSkyDiving()
   > Set Component Linear Acc Scale(Component Linear Acc Scale:Vector=SelectVector(A="0.5,0.5,0.5", B="0.3,0.3,0.3", bPickA=@n1)) -> (Output_Get:Vector)
   > Set Component Linear Vel Scale(Component Linear Vel Scale:Vector=SelectVector(A="0.5,0.5,0.5", B="0.3,0.3,0.3", bPickA=@n1)) -> (Output_Get:Vector)
   TryGetPawnOwner()
  ]}
]
variables [
 bIsSkyDivingOrDBNOCarried {type=boolean}
 Component Linear Acc Scale {type=Vector, default=<1 1 1>}
 Component Linear Vel Scale {type=Vector, default=<1 1 1>}
]
